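# Post-process style AnimBP for the CyclopsPrey male medium cosmetic body, adding secondary motion to its apron and ties.
# Copies the parent mesh pose through a Control Rig, then runs AnimDynamics chains, with constraints on tie bones.
# Additive bone offsets then correct the pelvis, crouch and apron bones, some blended by jumping/falling or non-targeting movement.
# EventGraph updates those alphas, an LOD morph, plus gravity and sim-space variables the dynamics nodes appear not to read.
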AnimBlueprint /BRCosmetics/Characters/Player/Male/Medium/Bodies/M_MED_CyclopsPrey/Meshes/M_MED_CyclopsPrey_AnimBP
target_skeleton=&/M_MED_CyclopsPrey_Skeleton
graphs [
 AnimGraph {nodes [
   Output Pose(Result=Use cached pose 'PostFixesCache'()) {layerGroup=None}
   n1 = Control Rig!(
    Source=Copy Pose From Mesh() {bUseAttachedParent=true}
    Alpha
    Is Playing Emote:bool=Get bIsPlayingEmote()
    isDBNO:bool=Get bIsDBNO()
    Is On Grind Rail:bool=Get bIsOnGrindRail()
   ) -> (Pose) {
    alphaBoolBlend {blendOption=Linear}
    controlRigAssetReference {blueprintRigClass=&/M_MED_CyclopsPrey_CtrlRig.M_MED_CyclopsPrey_CtrlRig_C}
   }
   Save cached pose 'StartupCache'(Pose=@n1) {cacheName=StartupCache}
   n2 = Get Jumping() -> (Jumping:double)
   n3 = "Transform (Modify) Bone - Bone: dyn_apron_fr_r_base"(ComponentPose=Local To Component(LocalPose=Use cached pose 'RbanCache'()), Alpha=@n2, Translation="0,0,0", Rotation="-10,0,0", Scale="1,1,1") {
    rotationMode=BMM_Additive
    translationMode=BMM_Additive
    alphaBoolBlend {blendOption=Linear}
    alphaScaleBiasClamp {bInterpResult=true, interpSpeedDecreasing=7, interpSpeedIncreasing=5}
    boneToModify {boneName=dyn_apron_fr_r_base}
    rotation {pitch=-10}
   }
   n4 = "Transform (Modify) Bone - Bone: dyn_apron_side_r_base"(ComponentPose=@n3, Alpha=@n2, Translation="0,0,0", Rotation="-15,0,0", Scale="1,1,1") {
    rotationMode=BMM_Additive
    translationMode=BMM_Additive
    alphaBoolBlend {blendOption=Linear}
    alphaScaleBiasClamp {bInterpResult=true, interpSpeedDecreasing=7, interpSpeedIncreasing=5}
    boneToModify {boneName=dyn_apron_side_r_base}
    rotation {pitch=-15}
   }
   n5 = "Transform (Modify) Bone - Bone: dyn_apron_fr_l_02"(ComponentPose=@n4, Alpha=@n2, Translation="0,0,0", Rotation="0.000,0.000,-25.000", Scale="1,1,1") {
    rotationMode=BMM_Additive
    translationMode=BMM_Additive
    alphaBoolBlend {blendOption=Linear}
    alphaScaleBiasClamp {bInterpResult=true, interpSpeedDecreasing=7, interpSpeedIncreasing=5}
    boneToModify {boneName=dyn_apron_fr_l_02}
    rotation {roll=-25}
   }
   n6 = "Transform (Modify) Bone - Bone: dyn_apron_side_l_02"(ComponentPose=@n5, Alpha=@n2, Translation="0,0,0", Rotation="10.000,-0.000,-20.000", Scale="1,1,1") {
    rotationMode=BMM_Additive
    translationMode=BMM_Additive
    alphaBoolBlend {blendOption=Linear}
    alphaScaleBiasClamp {bInterpResult=true, interpSpeedDecreasing=7, interpSpeedIncreasing=5}
    boneToModify {boneName=dyn_apron_side_l_02}
    rotation {pitch=10, roll=-20}
   }
   n7 = "Transform (Modify) Bone - Bone: dyn_apron_side_r_base"(ComponentPose=@n6, Alpha=Get Moving2DNotTarget() -> (Moving2DNotTarget:double), Translation="-4,0,0", Rotation="0,0,0", Scale="1,1,1") {
    rotationMode=BMM_Additive
    translationMode=BMM_Additive
    alphaBoolBlend {blendOption=Linear}
    alphaScaleBiasClamp {bInterpResult=true, interpSpeedDecreasing=7, interpSpeedIncreasing=5}
    boneToModify {boneName=dyn_apron_side_r_base}
    translation {x=-4}
   }
   n8 = "Transform (Modify) Bone - Bone: dfrm_pelvis"(ComponentPose=@n7, Alpha=1, Translation="-5,-15,0", Rotation="0,20,0", Scale="1,1,1") {
    rotationMode=BMM_Additive
    rotationSpace=BCS_BoneSpace
    translationMode=BMM_Additive
    translationSpace=BCS_BoneSpace
    alphaBoolBlend {blendOption=Linear}
    alphaScaleBiasClamp {bInterpResult=true, interpSpeedIncreasing=6}
    boneToModify {boneName=dfrm_pelvis}
    rotation {yaw=20}
    translation=<-5 -15>
   }
   n9 = "Transform (Modify) Bone - Bone: dfrm_crouch"(ComponentPose=@n8, Alpha=1, Translation="10,0,0", Rotation="0,20,0", Scale="1,1,1") {
    rotationMode=BMM_Additive
    rotationSpace=BCS_BoneSpace
    translationMode=BMM_Additive
    translationSpace=BCS_BoneSpace
    alphaBoolBlend {blendOption=Linear}
    alphaScaleBiasClamp {bInterpResult=true, interpSpeedIncreasing=6}
    boneToModify {boneName=dfrm_crouch}
    rotation {yaw=20}
    translation {x=10}
   }
   n10 = "Transform (Modify) Bone - Bone: dyn_apron_fr_c_base"(ComponentPose=@n9, Alpha=1, Translation="0,-2,0", Rotation="0,0,0", Scale="1,1,1") {
    rotationMode=BMM_Additive
    rotationSpace=BCS_BoneSpace
    translationMode=BMM_Additive
    translationSpace=BCS_BoneSpace
    alphaBoolBlend {blendOption=Linear}
    alphaScaleBiasClamp {bInterpResult=true, interpSpeedIncreasing=6}
    boneToModify {boneName=dyn_apron_fr_c_base}
    translation {y=-2}
   }
   Save cached pose 'PostFixesCache'(Pose=Component To Local(ComponentPose=@n10)) {cacheName=PostFixesCache}
   n11 = Anim Dynamics - Bone: dyn_apron_fr_r_02!(
    SimulationSpace:AnimPhysSimSpaceType Enum=BoneRelative
    ComponentPose=Local To Component(LocalPose=Use cached pose 'StartupCache'())
    Alpha=1
    GravityOverride:Vector="0,0,-1500"
    ExternalForce="0,0,0"
   ) -> (Pose) {
    angularDampingOverride=1
    angularSpringConstant=3000
    bAngularSpring=true
    bOverrideAngularDamping=true
    bUseGravityOverride=true
    gravityScale=0.35
    simulationSpace=BoneRelative
    alphaBoolBlend {blendOption=Linear}
    boundBone {boneName=dyn_apron_fr_r_02}
    componentAppliedLinearAccClamp=<100 100 100>
    componentLinearAccScale=<1 1 1>
    componentLinearVelScale=<1 1 1>
    gravityOverride {z=-1500}
    physicsBodyDefinitions [{
      boundBone {boneName=dyn_apron_fr_r_02}
      boxExtents=<17 17 17>
      constraintSetup {angularLimitsMax=<0 -5 0>, angularLimitsMin=<0 -15 0>, angularTarget=<1 0 0>}
      localJointOffset=<2 0 0>
     }]
    relativeSpaceBone {boneName=root}
   }
   n12 = Anim Dynamics - Bone: dyn_apron_fr_r_03!(
    SimulationSpace:AnimPhysSimSpaceType Enum=BoneRelative
    ComponentPose=@n11
    Alpha=1
    GravityOverride:Vector="0,0,-1500"
    ExternalForce="0,0,0"
   ) -> (Pose) {
    angularDampingOverride=1
    angularSpringConstant=3000
    bAngularSpring=true
    bOverrideAngularDamping=true
    bUseGravityOverride=true
    gravityScale=0.35
    simulationSpace=BoneRelative
    alphaBoolBlend {blendOption=Linear}
    boundBone {boneName=dyn_apron_fr_r_03}
    componentAppliedLinearAccClamp=<100 100 100>
    componentLinearAccScale=<1 1 1>
    componentLinearVelScale=<1 1 1>
    gravityOverride {z=-1500}
    physicsBodyDefinitions [{
      boundBone {boneName=dyn_apron_fr_r_03}
      boxExtents=<17 17 17>
      constraintSetup {angularLimitsMax=<0 -5 0>, angularLimitsMin=<0 -25 0>, angularTarget=<1 0 0>}
      localJointOffset=<2 0 0>
     }]
    relativeSpaceBone {boneName=root}
   }
   n13 = Anim Dynamics - Bone: dyn_apron_fr_r_04!(
    SimulationSpace:AnimPhysSimSpaceType Enum=BoneRelative
    ComponentPose=@n12
    Alpha=1
    GravityOverride:Vector="0,0,-1500"
    ExternalForce="0,0,0"
   ) -> (Pose) {
    angularDampingOverride=1
    angularSpringConstant=3000
    bAngularSpring=true
    bOverrideAngularDamping=true
    bUseGravityOverride=true
    gravityScale=0.35
    simulationSpace=BoneRelative
    alphaBoolBlend {blendOption=Linear}
    boundBone {boneName=dyn_apron_fr_r_04}
    componentAppliedLinearAccClamp=<100 100 100>
    componentLinearAccScale=<1 1 1>
    componentLinearVelScale=<1 1 1>
    gravityOverride {z=-1500}
    physicsBodyDefinitions [{
      boundBone {boneName=dyn_apron_fr_r_04}
      boxExtents=<17 17 17>
      constraintSetup {angularLimitsMax=<0 0 0>, angularLimitsMin=<0 -50 0>, angularTarget=<1 0 0>}
      localJointOffset=<2 0 0>
     }]
    relativeSpaceBone {boneName=root}
   }
   n14 = Anim Dynamics - Bone: dyn_apron_side_r_02!(
    SimulationSpace:AnimPhysSimSpaceType Enum=BoneRelative
    ComponentPose=@n13
    Alpha=1
    GravityOverride:Vector="0,0,-1500"
    ExternalForce="0,0,0"
   ) -> (Pose) {
    angularDampingOverride=1
    angularSpringConstant=3000
    bAngularSpring=true
    bOverrideAngularDamping=true
    bUseGravityOverride=true
    gravityScale=0.35
    simulationSpace=BoneRelative
    alphaBoolBlend {blendOption=Linear}
    boundBone {boneName=dyn_apron_side_r_02}
    componentAppliedLinearAccClamp=<100 100 100>
    componentLinearAccScale=<1 1 1>
    componentLinearVelScale=<1 1 1>
    gravityOverride {z=-1500}
    physicsBodyDefinitions [{
      boundBone {boneName=dyn_apron_side_r_02}
      boxExtents=<17 17 17>
      constraintSetup {angularLimitsMax=<0 0 10>, angularLimitsMin=<0 0 0>, angularTarget=<1 0 0>}
      localJointOffset=<2 0 0>
     }]
    relativeSpaceBone {boneName=root}
   }
   n15 = Anim Dynamics - Bone: dyn_apron_side_r_03!(
    SimulationSpace:AnimPhysSimSpaceType Enum=BoneRelative
    ComponentPose=@n14
    Alpha=1
    GravityOverride:Vector="0,0,-1500"
    ExternalForce="0,0,0"
   ) -> (Pose) {
    angularDampingOverride=1
    angularSpringConstant=3000
    bAngularSpring=true
    bOverrideAngularDamping=true
    bUseGravityOverride=true
    gravityScale=0.35
    simulationSpace=BoneRelative
    alphaBoolBlend {blendOption=Linear}
    boundBone {boneName=dyn_apron_side_r_03}
    componentAppliedLinearAccClamp=<100 100 100>
    componentLinearAccScale=<1 1 1>
    componentLinearVelScale=<1 1 1>
    gravityOverride {z=-1500}
    physicsBodyDefinitions [{
      boundBone {boneName=dyn_apron_side_r_03}
      boxExtents=<17 17 17>
      constraintSetup {angularLimitsMax=<0 0 30>, angularLimitsMin=<0 -5 0>, angularTarget=<1 0 0>}
      localJointOffset=<2 0 0>
     }]
    relativeSpaceBone {boneName=root}
   }
   n16 = Anim Dynamics - Bone: dyn_apron_side_r_04!(
    SimulationSpace:AnimPhysSimSpaceType Enum=BoneRelative
    ComponentPose=@n15
    Alpha=1
    GravityOverride:Vector="0,0,-1500"
    ExternalForce="0,0,0"
   ) -> (Pose) {
    angularDampingOverride=1
    angularSpringConstant=3000
    bAngularSpring=true
    bOverrideAngularDamping=true
    bUseGravityOverride=true
    gravityScale=0.35
    simulationSpace=BoneRelative
    alphaBoolBlend {blendOption=Linear}
    boundBone {boneName=dyn_apron_side_r_04}
    componentAppliedLinearAccClamp=<100 100 100>
    componentLinearAccScale=<1 1 1>
    componentLinearVelScale=<1 1 1>
    gravityOverride {z=-1500}
    physicsBodyDefinitions [{
      boundBone {boneName=dyn_apron_side_r_04}
      boxExtents=<17 17 17>
      constraintSetup {angularLimitsMax=<0 0 50>, angularLimitsMin=<0 -20 0>, angularTarget=<1 0 0>}
      localJointOffset=<2 0 0>
     }]
    relativeSpaceBone {boneName=root}
   }
   n17 = Anim Dynamics - Bone: dyn_apron_fr_c_02!(SimulationSpace:AnimPhysSimSpaceType Enum, ComponentPose=@n16, Alpha=1, GravityOverride:Vector="0,0,-1500", ExternalForce="0,0,0") -> (Pose) {
    angularDampingOverride=1
    angularSpringConstant=3000
    bAngularSpring=true
    bOverrideAngularDamping=true
    bUseGravityOverride=true
    gravityScale=0.35
    simulationSpace=RootRelative
    alphaBoolBlend {blendOption=Linear}
    boundBone {boneName=dyn_apron_fr_c_02}
    componentAppliedLinearAccClamp=<100 100 100>
    componentLinearAccScale=<1 1 1>
    componentLinearVelScale=<1 1 1>
    gravityOverride {z=-1500}
    physicsBodyDefinitions [{
      boundBone {boneName=dyn_apron_fr_c_02}
      boxExtents=<17 17 17>
      constraintSetup {angularLimitsMax=<0 0 0>, angularLimitsMin=<0 -10 0>, angularTarget=<1 0 0>}
      localJointOffset=<2 0 0>
     }]
   }
   n18 = Anim Dynamics - Bone: dyn_apron_fr_c_03!(SimulationSpace:AnimPhysSimSpaceType Enum, ComponentPose=@n17, Alpha=1, GravityOverride:Vector="0,0,-1500", ExternalForce="0,0,0") -> (Pose) {
    angularDampingOverride=1
    angularSpringConstant=3000
    bAngularSpring=true
    bOverrideAngularDamping=true
    bUseGravityOverride=true
    gravityScale=0.35
    simulationSpace=RootRelative
    alphaBoolBlend {blendOption=Linear}
    boundBone {boneName=dyn_apron_fr_c_03}
    componentAppliedLinearAccClamp=<100 100 100>
    componentLinearAccScale=<1 1 1>
    componentLinearVelScale=<1 1 1>
    gravityOverride {z=-1500}
    physicsBodyDefinitions [{
      boundBone {boneName=dyn_apron_fr_c_03}
      boxExtents=<17 17 17>
      constraintSetup {angularLimitsMax=<0 0 0>, angularLimitsMin=<0 -20 0>, angularTarget=<1 0 0>}
      localJointOffset=<2 0 0>
     }]
   }
   n19 = Anim Dynamics - Bone: dyn_apron_fr_c_04!(SimulationSpace:AnimPhysSimSpaceType Enum, ComponentPose=@n18, Alpha=1, GravityOverride:Vector="0,0,-1500", ExternalForce="0,0,0") -> (Pose) {
    angularDampingOverride=1
    angularSpringConstant=3000
    bAngularSpring=true
    bOverrideAngularDamping=true
    bUseGravityOverride=true
    gravityScale=0.35
    simulationSpace=RootRelative
    alphaBoolBlend {blendOption=Linear}
    boundBone {boneName=dyn_apron_fr_c_04}
    componentAppliedLinearAccClamp=<100 100 100>
    componentLinearAccScale=<1 1 1>
    componentLinearVelScale=<1 1 1>
    gravityOverride {z=-1500}
    physicsBodyDefinitions [{
      boundBone {boneName=dyn_apron_fr_c_04}
      boxExtents=<17 17 17>
      constraintSetup {angularLimitsMax=<0 5 0>, angularLimitsMin=<0 -50 0>, angularTarget=<1 0 0>}
      localJointOffset=<2 0 0>
     }]
   }
   n20 = Anim Dynamics - Bone: dyn_apron_fr_l_02!(SimulationSpace:AnimPhysSimSpaceType Enum, ComponentPose=@n19, Alpha=1, GravityOverride:Vector="0,0,-1500", ExternalForce="0,0,0") -> (Pose) {
    angularDampingOverride=1
    angularSpringConstant=3000
    bAngularSpring=true
    bOverrideAngularDamping=true
    bUseGravityOverride=true
    gravityScale=0.35
    simulationSpace=RootRelative
    alphaBoolBlend {blendOption=Linear}
    boundBone {boneName=dyn_apron_fr_l_02}
    componentAppliedLinearAccClamp=<100 100 100>
    componentLinearAccScale=<1 1 1>
    componentLinearVelScale=<1 1 1>
    gravityOverride {z=-1500}
    physicsBodyDefinitions [{
      boundBone {boneName=dyn_apron_fr_l_02}
      boxExtents=<17 17 17>
      constraintSetup {angularLimitsMax=<0 -5 0>, angularLimitsMin=<0 -10 0>, angularTarget=<1 0 0>}
      localJointOffset=<2 0 0>
     }]
   }
   n21 = Anim Dynamics - Bone: dyn_apron_fr_l_03!(SimulationSpace:AnimPhysSimSpaceType Enum, ComponentPose=@n20, Alpha=1, GravityOverride:Vector="0,0,-1500", ExternalForce="0,0,0") -> (Pose) {
    angularDampingOverride=1
    angularSpringConstant=3000
    bAngularSpring=true
    bOverrideAngularDamping=true
    bUseGravityOverride=true
    gravityScale=0.35
    simulationSpace=RootRelative
    alphaBoolBlend {blendOption=Linear}
    boundBone {boneName=dyn_apron_fr_l_03}
    componentAppliedLinearAccClamp=<100 100 100>
    componentLinearAccScale=<1 1 1>
    componentLinearVelScale=<1 1 1>
    gravityOverride {z=-1500}
    physicsBodyDefinitions [{
      boundBone {boneName=dyn_apron_fr_l_03}
      boxExtents=<17 17 17>
      constraintSetup {angularLimitsMax=<0 -5 0>, angularLimitsMin=<0 -10 0>, angularTarget=<1 0 0>}
      localJointOffset=<2 0 0>
     }]
   }
   n22 = Anim Dynamics - Bone: dyn_apron_fr_l_04!(SimulationSpace:AnimPhysSimSpaceType Enum, ComponentPose=@n21, Alpha=1, GravityOverride:Vector="0,0,-1500", ExternalForce="0,0,0") -> (Pose) {
    angularDampingOverride=1
    angularSpringConstant=3000
    bAngularSpring=true
    bOverrideAngularDamping=true
    bUseGravityOverride=true
    gravityScale=0.35
    simulationSpace=RootRelative
    alphaBoolBlend {blendOption=Linear}
    boundBone {boneName=dyn_apron_fr_l_04}
    componentAppliedLinearAccClamp=<100 100 100>
    componentLinearAccScale=<1 1 1>
    componentLinearVelScale=<1 1 1>
    gravityOverride {z=-1500}
    physicsBodyDefinitions [{
      boundBone {boneName=dyn_apron_fr_l_04}
      boxExtents=<17 17 17>
      constraintSetup {angularLimitsMax=<0 0 0>, angularLimitsMin=<0 -50 0>, angularTarget=<1 0 0>}
      localJointOffset=<2 0 0>
     }]
   }
   n23 = Anim Dynamics - Bone: dyn_apron_side_l_02!(SimulationSpace:AnimPhysSimSpaceType Enum, ComponentPose=@n22, Alpha=1, GravityOverride:Vector="0,0,-1500", ExternalForce="0,0,0") -> (Pose) {
    angularDampingOverride=1
    angularSpringConstant=3000
    bAngularSpring=true
    bOverrideAngularDamping=true
    bUseGravityOverride=true
    gravityScale=0.35
    simulationSpace=RootRelative
    alphaBoolBlend {blendOption=Linear}
    boundBone {boneName=dyn_apron_side_l_02}
    componentAppliedLinearAccClamp=<100 100 100>
    componentLinearAccScale=<1 1 1>
    componentLinearVelScale=<1 1 1>
    gravityOverride {z=-1500}
    physicsBodyDefinitions [{
      boundBone {boneName=dyn_apron_side_l_02}
      boxExtents=<17 17 17>
      constraintSetup {angularLimitsMax=<0 -5 0>, angularLimitsMin=<0 -10 0>, angularTarget=<1 0 0>}
      localJointOffset=<2 0 0>
     }]
   }
   n24 = Anim Dynamics - Bone: dyn_apron_side_l_03!(SimulationSpace:AnimPhysSimSpaceType Enum, ComponentPose=@n23, Alpha=1, GravityOverride:Vector="0,0,-1500", ExternalForce="0,0,0") -> (Pose) {
    angularDampingOverride=1
    angularSpringConstant=3000
    bAngularSpring=true
    bOverrideAngularDamping=true
    bUseGravityOverride=true
    gravityScale=0.35
    simulationSpace=RootRelative
    alphaBoolBlend {blendOption=Linear}
    boundBone {boneName=dyn_apron_side_l_03}
    componentAppliedLinearAccClamp=<100 100 100>
    componentLinearAccScale=<1 1 1>
    componentLinearVelScale=<1 1 1>
    gravityOverride {z=-1500}
    physicsBodyDefinitions [{
      boundBone {boneName=dyn_apron_side_l_03}
      boxExtents=<17 17 17>
      constraintSetup {angularLimitsMax=<0 -5 0>, angularLimitsMin=<0 -5 -10>, angularTarget=<1 0 0>}
      localJointOffset=<2 0 0>
     }]
   }
   n25 = Anim Dynamics - Bone: dyn_apron_side_l_04!(SimulationSpace:AnimPhysSimSpaceType Enum, ComponentPose=@n24, Alpha=1, GravityOverride:Vector="0,0,-1500", ExternalForce="0,0,0") -> (Pose) {
    angularDampingOverride=1
    angularSpringConstant=3000
    bAngularSpring=true
    bOverrideAngularDamping=true
    bUseGravityOverride=true
    gravityScale=0.35
    simulationSpace=RootRelative
    alphaBoolBlend {blendOption=Linear}
    boundBone {boneName=dyn_apron_side_l_04}
    componentAppliedLinearAccClamp=<100 100 100>
    componentLinearAccScale=<1 1 1>
    componentLinearVelScale=<1 1 1>
    gravityOverride {z=-1500}
    physicsBodyDefinitions [{
      boundBone {boneName=dyn_apron_side_l_04}
      boxExtents=<17 17 17>
      constraintSetup {angularLimitsMax=<0 -5 0>, angularLimitsMin=<0 -20 -50>, angularTarget=<1 0 0>}
      localJointOffset=<2 0 0>
     }]
   }
   n26 = "Constraint - dyn_tie_l_loop (PP)"!(ComponentPose=@n25, Alpha=1, ConstraintWeights_0=1, ConstraintWeights_1:float=0) -> (Pose) {
    alphaBoolBlend {blendOption=Linear}
    boneToModify {boneName=dyn_tie_l_loop}
    constraintSetup [{targetBone {boneName=dyn_tie_base}}, {targetBone {boneName=thigh_l}}]
    constraintWeights [1,1]
   }
   n27 = "Constraint - dyn_tie_l_string (PP)"!(ComponentPose=@n26, Alpha=1, ConstraintWeights_0=1, ConstraintWeights_1:float=0) -> (Pose) {
    alphaBoolBlend {blendOption=Linear}
    boneToModify {boneName=dyn_tie_l_string}
    constraintSetup [{targetBone {boneName=dyn_tie_base}}, {targetBone {boneName=thigh_l}}]
    constraintWeights [1,1]
   }
   n28 = "Constraint - dyn_tie_r_loop (PP)"!(ComponentPose=@n27, Alpha=1, ConstraintWeights_0=1, ConstraintWeights_1:float=0) -> (Pose) {
    alphaBoolBlend {blendOption=Linear}
    boneToModify {boneName=dyn_tie_r_loop}
    constraintSetup [{targetBone {boneName=dyn_tie_base}}, {targetBone {boneName=thigh_l}}]
    constraintWeights [1,1]
   }
   n29 = "Constraint - dyn_tie_r_string (PP)"!(ComponentPose=@n28, Alpha=1, ConstraintWeights_0=1, ConstraintWeights_1:float=0) -> (Pose) {
    alphaBoolBlend {blendOption=Linear}
    boneToModify {boneName=dyn_tie_r_string}
    constraintSetup [{targetBone {boneName=dyn_tie_base}}, {targetBone {boneName=thigh_l}}]
    constraintWeights [1,1]
   }
   n30 = Anim Dynamics - Bone: dyn_tie_r_loop!(
    SimulationSpace:AnimPhysSimSpaceType Enum=BoneRelative
    ComponentPose=@n29
    Alpha=1
    GravityOverride:Vector="0,0,-1500"
    ExternalForce="0,0,0"
   ) -> (Pose) {
    angularDampingOverride=0.999
    angularSpringConstant=1000
    bAngularSpring=true
    bOverrideAngularDamping=true
    bUseGravityOverride=true
    gravityScale=0.35
    simulationSpace=BoneRelative
    alphaBoolBlend {blendOption=Linear}
    boundBone {boneName=dyn_tie_r_loop}
    componentAppliedLinearAccClamp=<100 100 100>
    componentLinearAccScale=<1 1 1>
    componentLinearVelScale=<1 1 1>
    gravityOverride {z=-1500}
    physicsBodyDefinitions [{
      boundBone {boneName=dyn_tie_r_loop}
      boxExtents=<12 12 12>
      constraintSetup {angularLimitsMax=<0 0 20>, angularLimitsMin=<0 -30 -20>, angularTarget=<1 0 0>}
      localJointOffset=<1 0 0>
     }]
    relativeSpaceBone {boneName=root}
   }
   n31 = Anim Dynamics - Bone: dyn_tie_l_loop!(
    SimulationSpace:AnimPhysSimSpaceType Enum=BoneRelative
    ComponentPose=@n30
    Alpha=1
    GravityOverride:Vector="0,0,-1500"
    ExternalForce="0,0,0"
   ) -> (Pose) {
    angularDampingOverride=0.999
    angularSpringConstant=1000
    bAngularSpring=true
    bOverrideAngularDamping=true
    bUseGravityOverride=true
    gravityScale=0.35
    simulationSpace=BoneRelative
    alphaBoolBlend {blendOption=Linear}
    boundBone {boneName=dyn_tie_l_loop}
    componentAppliedLinearAccClamp=<100 100 100>
    componentLinearAccScale=<1 1 1>
    componentLinearVelScale=<1 1 1>
    gravityOverride {z=-1500}
    physicsBodyDefinitions [{
      boundBone {boneName=dyn_tie_l_loop}
      boxExtents=<12 12 12>
      constraintSetup {angularLimitsMax=<0 0 20>, angularLimitsMin=<0 -30 -20>, angularTarget=<1 0 0>}
      localJointOffset=<1 0 0>
     }]
    relativeSpaceBone {boneName=root}
   }
   n32 = Anim Dynamics - Bone: dyn_tie_r_string!(
    SimulationSpace:AnimPhysSimSpaceType Enum=BoneRelative
    ComponentPose=@n31
    Alpha=1
    GravityOverride:Vector="0,0,-1500"
    ExternalForce="0,0,0"
   ) -> (Pose) {
    angularDampingOverride=0.999
    angularSpringConstant=1000
    bAngularSpring=true
    bOverrideAngularDamping=true
    bUseGravityOverride=true
    gravityScale=0.35
    simulationSpace=BoneRelative
    alphaBoolBlend {blendOption=Linear}
    boundBone {boneName=dyn_tie_r_string}
    componentAppliedLinearAccClamp=<100 100 100>
    componentLinearAccScale=<1 1 1>
    componentLinearVelScale=<1 1 1>
    gravityOverride {z=-1500}
    physicsBodyDefinitions [{
      boundBone {boneName=dyn_tie_r_string}
      boxExtents=<12 12 12>
      constraintSetup {angularLimitsMax=<0 30 20>, angularLimitsMin=<0 -30 -20>, angularTarget=<1 0 0>}
      localJointOffset=<1 0 0>
     }]
    relativeSpaceBone {boneName=root}
   }
   n33 = Anim Dynamics - Bone: dyn_tie_l_string!(
    SimulationSpace:AnimPhysSimSpaceType Enum=BoneRelative
    ComponentPose=@n32
    Alpha=1
    GravityOverride:Vector="0,0,-1500"
    ExternalForce="0,0,0"
   ) -> (Pose) {
    angularDampingOverride=0.999
    angularSpringConstant=1000
    bAngularSpring=true
    bOverrideAngularDamping=true
    bUseGravityOverride=true
    gravityScale=0.35
    simulationSpace=BoneRelative
    alphaBoolBlend {blendOption=Linear}
    boundBone {boneName=dyn_tie_l_string}
    componentAppliedLinearAccClamp=<100 100 100>
    componentLinearAccScale=<1 1 1>
    componentLinearVelScale=<1 1 1>
    gravityOverride {z=-1500}
    physicsBodyDefinitions [{
      boundBone {boneName=dyn_tie_l_string}
      boxExtents=<12 12 12>
      constraintSetup {angularLimitsMax=<0 30 20>, angularLimitsMin=<0 -30 -20>, angularTarget=<1 0 0>}
      localJointOffset=<1 0 0>
     }]
    relativeSpaceBone {boneName=root}
   }
   Save cached pose 'RbanCache'(Pose=Component To Local(ComponentPose=@n33)) {cacheName=RbanCache}
  ]}
 EventGraph {nodes [
   n1 = Event BlueprintUpdateAnimation()
   > Set DeltaTimeLocal(DeltaTimeLocal:double=@n1.DeltaTimeX) -> (Output_Get:double)
   > n2 = Sequence!(execute) -> (then_0, then_1, then_2:Exec, then_3:Exec, then_4:Exec)
   > n3 = Branch(Condition=Get bClothEnabled())
   > n4 = Sequence!(execute) -> (then_0, then_1)
   n5 = Gravity Override State Function DEPRECATED(
    execute=[@n4.then_0, @n3.else]
    CharacterPartAnimInstance=Self-Reference()
    GameTimeInSeconds=GetGameTimeInSeconds()
    GravityOverrideParams=Get Gravity Override Params()
   )
   > Set GravityOverride(GravityOverride=@n5)
   Set simSpace(execute=@n2.then_3, simSpace:AnimPhysSimSpaceType Enum=BoneRelative) -> (Output_Get:AnimPhysSimSpaceType Enum)
   > Branch(Condition=OR Boolean(A=Get bIsDBNO(), B=Get bIsInFrontEnd()))
   > n6 = Branch(Condition=OR Boolean(A=Get bIsDBNOCarried(), B=Get bIsInFrontEnd()))
   > Set simSpace(simSpace=World)
   Set simSpace(execute=@n6.else)
   n7 = Branch(execute=@n2.then_2, Condition="integer > integer"(A=Get CurrentLOD(), B=0) {is=title K2Node_CallFunction})
   > SetMorphTarget!(execute, self:Anim Instance Object Reference, MorphTargetName:Name=Body_LOD_BS, Value:float=1) -> (then)
   SetMorphTarget(execute=@n7.else, MorphTargetName=Body_LOD_BS, Value=0)
   Set frontendAlpha(execute=@n2.then_4, frontendAlpha:double=SelectFloat(A=0.5, B=1, bPickA=Get bIsInFrontEnd())) -> (Output_Get:double)
   > n8 = Set isDBNO_Carried(isDBNO_Carried:double=SelectFloat(A=1, B=0, bPickA=Get bIsDBNOCarried())) -> (Output_Get:double)
   > Set frontendAlpha(frontendAlpha=float - float(A=Get frontendAlpha() -> (frontendAlpha:double), B=@n8))
   n9 = NOT Boolean(A=Get bIsTargeting())
   n10 = NOT Boolean(A=Get bIsCrouching())
   n11 = Get bIsMoving2D()
   > Set Moving2DNotTarget(Moving2DNotTarget:double=SelectFloat(A=1, B=0, bPickA=AND Boolean!(A=@n11, B=@n9, C:bool=@n10) -> (ReturnValue))) -> (Output_Get:double)
   > Set EmoteAlpha(EmoteAlpha:double=SelectFloat(A=1, B=0, bPickA=Get bIsPlayingEmote())) -> (Output_Get:double)
   Event OnBeginCrouching()
   > Set Is Crouching Alpha(Is Crouching Alpha:double=1) -> (Output_Get:double)
   Event OnEndCrouching()
   > Set Is Crouching Alpha(Is Crouching Alpha=0)
   Event OnBeginFalling()
   > Set Jumping(Jumping:double=1) -> (Output_Get:double)
   Event OnEndFalling()
   > Set Jumping(Jumping=0)
   Event OnBeginJumping()
   > Set Jumping(Jumping=1)
   Event OnEndJumping()
   > Set Jumping(Jumping=0)
   Get bIsSprinting()
  ]}
]
variables [
 DeltaTimeLocal {type=number}
 EmoteAlpha {type=number}
 frontendAlpha {type=number}
 Gravity Override Params {type=GravityOverrideParamsStruct}
 GravityOverride {type=Vector}
 Is Crouching Alpha {type=number}
 isDBNO_Carried {type=number}
 Jumping {type=number}
 Moving2DNotTarget {type=number}
 simSpace {type=AnimPhysSimSpaceType}
]
